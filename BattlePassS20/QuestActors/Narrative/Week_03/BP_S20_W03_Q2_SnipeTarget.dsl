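; Quest target actor for a Battle Pass Season 20 Week 3 snipe quest, which credits hits only from long range.
; CheckTagsAndComplete defers to the parent's completion logic only when the player's pawn is at least 7500 units from the target.
; The construction script removes the inherited discover-radius collision, and visual updates instead call OnEnterCosmetic for the local player on remote clients.
; When the quest is not ready, it shows post-interaction visuals without VFX or audio.
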(event EventBeginPlay
  (|CallParentFunction) ; This node is disabled and will not be called.\nDrag off pins to build functionality.
)

(event EventUpdateVisuals
  (params (PlayerController :type Fort%20Player%20Controller%20Object%20Reference))
  (bind _returnvalue (Game|Player|GetPlayerPawn PlayerController))
  (Networking|SwitchHasAuthority
    (:Remote
      (Utilities|IsValid PlayerController
        (:Is%20Valid
          (if (Pawn|IsLocalPlayerController PlayerController)
            (Utilities|IsValid _returnvalue
              (:Is%20Valid
                (bind _asfort_player_pawn_athena (Utilities|Casting|CastToFortPlayerPawnAthena _returnvalue))
                (CallFunction|OnEnterCosmetic PlayerController _asfort_player_pawn_athena)))))))))

(event Collision|EventActorBeginOverlap
  (params (OtherActor :type Actor%20Object%20Reference))
  (|CallParentFunction OtherActor) ; This node is disabled and will not be called.\nDrag off pins to build functionality.
)

(event EventTick
  (params (DeltaSeconds :type Float%20%28single-precision%29))
  (|CallParentFunction DeltaSeconds) ; This node is disabled and will not be called.\nDrag off pins to build functionality.
)

(event EventVisualTestingNoQuestRequired_Init)

(event OnQuestNotReady(ConsolidatedQuestComponent)
  (params (PlayerController :type Fort%20Player%20Controller%20Object%20Reference))
  (VisualState|Meshes|PostInteractionVisuals PlayerController)
  (VisualState|VFX&Audio|NoVFX&Audio PlayerController))

(fn CheckTagsAndComplete
  (params (Actor :type Actor%20Object%20Reference))
  (bind _asfort_player_controller (Utilities|Casting|CastToFortPlayerController Actor))
  (if (>= (Transformation|GetDistanceTo (Game|Player|GetPlayerPawn _asfort_player_controller) self) 7.5e+03)
    (|CallParentFunction Actor)))

(fn ConstructionScript
  (|CallParentFunction)
  (Components|DestroyComponent (Variables|Components|GetCollision_DiscoverRadius)))
