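; Cosmetic shooting target actor, named for a Season 20 Week 3 snipe quest, that plays a one-shot rotation animation and removes itself.
; On BeginPlay it adds an audio component and caches the target component's relative location and rotation.
; The TargetRotation timeline then drives the roll while keeping the cached position, pitch and yaw.
; The actor destroys itself when the timeline finishes.
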(event EventBeginPlay
  (bind (_initialrotation_roll _initialrotation_pitch _initialrotation_yaw) (Math|Rotator|BreakRotator (Variables|GetInitialRotation_Roll)))
  (bind _shooting_target_01 (Variables|Default|Getshooting_target_01))
  (AddComponent|NODE_AddAudioComponent-0)
  (Variables|Default|SetInitialPosition (Variables|/Script/Engine.SceneComponent|GetRelativeLocation _shooting_target_01))
  (Variables|Default|SetInitialRotation (Variables|/Script/Engine.SceneComponent|GetRelativeRotation _shooting_target_01))
  (bind (_direction _rotation) (Timeline|TargetRotation PlayFromStart)
    (:Update
      (Transformation|SetRelativeLocationAndRotation (Variables|Default|Getshooting_target_01) (Variables|Default|GetInitialPosition) (Math|Rotator|MakeRotator _rotation _initialrotation_pitch _initialrotation_yaw) false false))
    (:Finished
      (Actor|DestroyActor))))

(fn ConstructionScript)
