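; Season 20 Week 3 quest Data Transponder prop that refuses interaction while buildings block its signal beam overhead.
; First cosmetic-range entry starts an upward trace; the server also gives eligible players a blocked voiceover effect.
; The server binds the hit building's OnDied to retrace, while clients move the beam and impact effect to the blockage.
; An empty trace unlocks interaction, fades the beam and refreshes local players' visuals.
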(event Custom|InitCheckBlocked
  (Utilities|FlowControl|DoOnce false
    (:Completed
      (CallFunction|UpdateBlockedParticle 0))))

(event Custom|UpdateBlockedEmitterLocation
  (params (Hit%20Location :type Vector) (Hit%20Impact%20Normal :type Vector%20%28by%20ref%29))
  (bind _returnvalue (- Hit%20Location "0,0,30.000000"))
  (bind _blockedsignalniagara (Variables|Default|GetBlockedSignalNiagara))
  (Networking|SwitchHasAuthority
    (:Remote
      (Transformation|SetWorldTransform _blockedsignalniagara (Math|Transform|MakeTransform _returnvalue (Math|Rotator|MakeRotfromZ Hit%20Impact%20Normal)) false false)
      (Development|SetHiddeninGame _blockedsignalniagara true false)
      (bind (_direction _lerptimeline) (Timeline|LerpTimeline PlayFromStart)
        (:Update
          (Effects|Components|ParticleSystem|SetVectorParameter (Variables|Components|GetN_Ambient) "BeamEndLocation" (Math|Vector|Lerp(Vector) (.Location (Collision|BreakHitResult (Utilities|Array|Get(aref) (stale Custom|UpdateBlockedParticle _returnvalue)))) _returnvalue _lerptimeline)))
        (:Finished
          (Development|SetHiddeninGame (Variables|Default|GetBlockedSignalNiagara) false false))))))

(event Custom|UpdateBlockedParticle
  (params (IgnoredActor :type Actor%20Object%20Reference))
  (bind _outhits (Collision|MultiLineTraceForObjects (+ (Transformation|GetActorLocation) "0,0,100.000000") (+ (Transformation|GetActorLocation) "0,0,1920.000000") (Utilities|Array|MakeArray "ObjectTypeQuery1") false (Utilities|Array|MakeArray IgnoredActor)))
  (bind (_batleastoneactorleft _returnvalue) (Ability|FilterTargetHitResultList (Variables|Default|GetBuildingFilter) 0 _outhits))
  (if _batleastoneactorleft
    (Networking|SwitchHasAuthority
      (:Authority
        (bind _asbuilding_actor (Utilities|Casting|CastToBuildingActor (.HitActor (Collision|BreakHitResult (Utilities|Array|Get(aref) _returnvalue)))))
        (EventDispatchers|BindEventtoOnDied _asbuilding_actor (delegate Custom|OnDiedLinkedActor)))
      (:Remote
        _))
    (CallFunction|UpdateBlockedEmitterLocation (.Location (Collision|BreakHitResult (Utilities|Array|Get(aref) _returnvalue))) (.ImpactNormal (Collision|BreakHitResult (Utilities|Array|Get(aref) _returnvalue))))
    (else
      (Variables|Default|SetbBlockInteract false)
      (CallFunction|UpdateVisualsAllClients))))

(event Custom|OnDiedLinkedActor
  (params (DamagedActor :type Actor%20Object%20Reference) (Damage :type Float%20%28single-precision%29) (InstigatedBy :type Controller%20Object%20Reference) (DamageCauser :type Actor%20Object%20Reference) (HitLocation :type Vector) (FHitComponent :type Primitive%20Component%20Object%20Reference) (BoneName :type Name) (Momentum :type Vector))
  (CallFunction|UpdateBlockedParticle DamagedActor))

(event Custom|UpdateVisualsAllClients
  (bind (_direction _lerptimeline) (Timeline|Timeline_1 PlayFromStart)
    (:Update
      (Effects|Components|ParticleSystem|SetFloatParameter (Variables|Components|GetN_Ambient) "MasterAlpha" (Math|Float|Lerp 1.0 0.0 _lerptimeline)))
    (:Finished
      (Components|Activation|Deactivate (Variables|Default|GetBlockedSignalNiagara))))
  (Networking|SwitchHasAuthority
    (:Remote
      (for _array_element (Fortnite|GetLocalFortPlayerControllers true)
        (CallFunction|UpdateVisuals _array_element)))))

(event EventOnEnterCosmetic
  (params (PlayerController :type Fort%20Player%20Controller%20Object%20Reference) (PlayerPawn :type Fort%20Player%20Pawn%20Athena%20Object%20Reference))
  (|CallParentFunction PlayerController PlayerPawn)
  (CallFunction|InitCheckBlocked)
  (Networking|SwitchHasAuthority
    (:Authority
      (if (Math|Boolean|ANDBoolean (or (ObjectiveState|IsObjectiveActive PlayerController) (Variables|Design|QuestSetup|GetbTestVisualsNoQuestRequired)) (not (GameplayTags|TagContainer|CompareTagAssetInterfacetoOtherTags (Variables|Default|GetGrantedBlockedTag) PlayerPawn)) (Variables|Default|GetbBlockInteract))
        (Environment|ApplyGlobalEnvironmentGameplayEffecttoActor PlayerPawn "/BattlePassS20/Items/QuestAbilities/Voiceover/Week3/Q03/GE_DataTransponderBlocked.GE_DataTransponderBlocked_C")))
    (:Remote
      (bind _asfort_quest_item_definition_athena (Utilities|Casting|CastToFortQuestItemDefinition_Athena (Class|ConsolidatedQuestComponent|GetQuestItemDefinition (Variables|BuildingGameplayActorPropQuest|GetConsolidatedQuestComponent))))
      (bind (_playercontrollerout _completedquest) (Quest|HasPlayerCompletedQuest PlayerController _asfort_quest_item_definition_athena))
      (if (and (Variables|Default|GetbBlockInteract) (or (and (not _completedquest) (ObjectiveState|IsObjectiveActive _playercontrollerout)) (and (Variables|Design|QuestSetup|GetbTestVisualsNoQuestRequired) (not (Variables|Advanced|Debug|GetbOnVisualTestingCooldown)))))
        (Components|Activation|Activate (Variables|Components|GetN_Ambient) true)))))

(fn BlueprintCanInteract
  (params (InteractingPawn :type Fort%20Pawn%20Object%20Reference) (InteractionBeingAttempted :type EInteractionBeingAttempted%20Enum :default "FirstInteraction") (InteractionType :type TInteractionType%20Enum :default "IT_NoInteraction"))
  (return (and (|CallParentFunction InteractingPawn InteractionBeingAttempted InteractionType) (not (Variables|Default|GetbBlockInteract)))))

(fn NoVFX&Audio
  (params (PlayerController :type Fort%20Player%20Controller%20Object%20Reference))
  (|CallParentFunction PlayerController)
  (Components|Activation|Deactivate (Variables|Components|GetN_Ambient)))

(fn PostInteractionVFX&Audio
  (params (PlayerController :type Fort%20Player%20Controller%20Object%20Reference))
  (|CallParentFunction PlayerController)
  (Components|Activation|Deactivate (Variables|Components|GetN_Ambient)))

(fn PreInteractionVFX&Audio
  (params (PlayerController :type Fort%20Player%20Controller%20Object%20Reference))
  (bind _n_ambient (Variables|Components|GetN_Ambient))
  (|CallParentFunction PlayerController)
  (if (Variables|Default|GetbBlockInteract)
    (Components|Activation|Activate _n_ambient true)
    (else
      (Components|Activation|Deactivate _n_ambient))))

(fn ConstructionScript
  (bind _n_ambient (Variables|Components|GetN_Ambient))
  (|CallParentFunction)
  (Transformation|AddLocalOffset _n_ambient "0, 0,40.000000" false false)
  (Niagara|SetNiagaraVariable(Vector3) _n_ambient "BeamEndLocation" (+ (Transformation|GetActorLocation) "0,0,350.000000"))
  (Effects|Components|ParticleSystem|SetFloatParameter _n_ambient "MasterAlpha" 1.0)
  (Niagara|SetNiagaraSystemAsset (Variables|Default|GetBlockedSignalNiagara) "/BattlePassS20/QuestActors/Narrative/Week_03/VFX/NS_W03_Q3_Radiowave_Impact.NS_W03_Q3_Radiowave_Impact" true)
  (Audio|Components|Audio|SetSound (Variables|Default|GetBlockedAmbientAudio) "/BattlePassS20/Audio/SFX/Quest_Transmission_Loop_Cue.Quest_Transmission_Loop_Cue"))
