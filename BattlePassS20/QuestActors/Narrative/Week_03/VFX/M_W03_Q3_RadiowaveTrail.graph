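# Translucent unlit Niagara ribbon material for the radiowave trail VFX in a Battle Pass S20 week 3 quest.
# A V-based sphere mask bends the UVs, which feed static and panning GeneratedBand functions to build procedural moving wave bands without textures.
# The bands are masked by a U gradient, edge sphere masks, camera depth and view-angle fades.
# The result blends dark blue to cyan for emissive and scales with MasterAlpha for opacity.
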Material /BattlePassS20/QuestActors/Narrative/Week_03/VFX/M_W03_Q3_RadiowaveTrail
graph {
 properties {bUsedWithNiagaraRibbons=true, blendMode=BLEND_Translucent, pixelDepthOffsetMode=PDOM_Legacy, shadingModel=MSM_Unlit, usedShadingModels=Unlit}
 nodes [
  n1 = Constant()
  n2 = StaticBool() {value=true}
  n3 = TextureCoordinate() {uTiling=0}
  n4 = Constant2Vector() {g=0.5}
  n5 = Constant() {r=0.5}
  n6 = Constant()
  n7 = Power(Base=OneMinus(None=SphereMask(A=@n3, B=@n4, Radius=@n5, Hardness=@n6)))
  n8 = ComponentMask(None=TextureCoordinate()) {r=true}
  n9 = Add(A=Multiply(A=TextureCoordinate() {coordinateIndex=1}) {constB=6}, B=Add(A=@n7, B=OneMinus(None=@n8)))
  n10 = MaterialFunctionCall!(Width=Constant() {r=0.5}, Sharpness=@n1, Offset, Direction Switch=@n2, Compare, Input Coordinates=Multiply(A=@n9) {constB=10}) {materialFunction=/Engine/Functions/Engine_MaterialFunctions03/Procedurals/GeneratedBand}
  n11 = Panner(Coordinate=@n9, Speed=AppendVector(A=ScalarParameter() {defaultValue=-1.75, parameterName=PanningSpeed}, B=Constant())) {speedX=-1}
  n12 = MaterialFunctionCall(Width=Constant() {r=0.5}, Sharpness=@n1, Direction Switch=@n2, Input Coordinates=@n11) {materialFunction=&GeneratedBand}
  n13 = Add(
   A=Multiply(A=Multiply(A=@n10.Result, B=@n12.Result)) {constB=0.5}
   B=MaterialFunctionCall(Width=Constant() {r=0.05}, Sharpness=@n1, Direction Switch=@n2, Input Coordinates=@n11) {materialFunction=&GeneratedBand}.Result
  )
  n14 = ComponentMask(None=TextureCoordinate()) {r=true}
  n15 = Multiply(A=Add(A=OneMinus(None=@n14)) {constB=0.25}) {constB=0.25}
  n16 = Saturate(None=Multiply(A=Multiply(A=@n13, B=@n15)) {constB=4})
  n17 = MaterialFunctionCall(Fade Length=Constant() {r=512}, Fade Offset=Constant() {r=32}) {materialFunction=&CameraDepthFade}
  n18 = Constant() {r=1}
  n19 = TextureCoordinate() {uTiling=0}
  n20 = Constant2Vector() {g=0.5}
  n21 = Constant() {r=0.45}
  n22 = Constant()
  n23 = TextureCoordinate() {vTiling=0}
  n24 = Constant2Vector() {r=0.5}
  n25 = Constant() {r=0.5}
  n26 = Constant() {r=0.5}
  n27 = Multiply(
   A=Multiply(
    A=Multiply(A=@n16, B=@n17.Result)
    B=MaterialFunctionCall("FadeSensitivity (optional)"=@n18) {materialFunction=&FadeBasedOnViewAngle}.Result
   )
   B=Multiply(A=SphereMask(A=@n19, B=@n20, Radius=@n21, Hardness=@n22), B=SphereMask(A=@n23, B=@n24, Radius=@n25, Hardness=@n26))
  )
  MaterialFunctionCall(Emissive=LinearInterpolate(A=Constant3Vector() {constant=<rgba 0 0.0832 0.3073 0>}, B=Constant3Vector() {constant=<rgba 0 0.7222 0.7812 0>}, Alpha=@n27)) {materialFunction=&MF_FX_EmissiveAdjustments}
  Multiply(A=@n7) {constB=2}
  Multiply(A=ScalarParameter() {defaultValue=0.7, parameterName=MasterAlpha}, B=Multiply(A=@n27) {constB=0.5})
 ]
}
