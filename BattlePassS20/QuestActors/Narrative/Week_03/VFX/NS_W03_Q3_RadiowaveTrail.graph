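# Niagara beam effect for the radiowave trail in a Battle Pass S20 Week 3 narrative quest.
# A single Beam emitter bursts particles into a ribbon ending at the absolute User.BeamEndLocation, with width driven by a curve over ribbon link order.
# Particles are blue, loop their lifetime rather than expiring, and take alpha from a falling curve.
# User.MasterAlpha is copied to an emitter variable, with no visible consumer here.
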NiagaraSystem /BattlePassS20/QuestActors/Narrative/Week_03/VFX/NS_W03_Q3_RadiowaveTrail
dependencies {
 usedDataInterfaces [&NiagaraDataInterfaceCurve]
 usedDynamicInputs [&Multiply_VectorByFloat]
 usedModules [&SystemState]
}
emitters {Beam {
  emitterSpawnScript [/SetVariables(Emitter.MasterAlpha:NiagaraFloat=$User.MasterAlpha) {isSetParameters=true}]
  emitterUpdateScript [
   &EmitterState(Life Cycle Mode=::NewEnumerator1, Loop Behavior=::NewEnumerator1, Loop Duration Mode=::NewEnumerator1, UseLoopDelay=false)
   /Niagara/Modules/Beams/BeamEmitterSetup(Absolute Beam End=true, Beam End:Vector3f=$User.BeamEndLocation)
   &SpawnBurst_Instantaneous(Spawn Count=200, Use Spawn Probability=false, UseLoopCountLimit=false)
  ]
  particleSpawnScript [
   &InitializeParticle(Color Mode=::NewEnumerator0, Lifetime=3, Position Mode=::NewEnumerator0, Write Material Random=true)
   /Niagara/Modules/Beams/SpawnBeam()
   /Niagara/Modules/Beams/BeamWidth(Beam Width:NiagaraFloat=&FloatFromCurve(
     FloatCurve=DataInterface() {propertyValues {Curve {keys [{tangentMode=RCTM_User, time=0, value=0.2}, {tangentMode=RCTM_User, time=1, value=1}]}}}
     CurveIndex=$Particles.RibbonLinkOrder
     Scale Curve=500
    ))
  ]
  particleUpdateScript [
   &ParticleState(
    Kill Particles When Lifetime Has Elapsed=false
    Let Infinitely Lived Particles Die When Emitter Deactivates=true
    Loop Particles Lifetime=true
   )
   /Niagara/Modules/Beams/UpdateBeam()
   &Color(
    Color=<rgba 0 0.4385 1 1>
    Scale Alpha=&FloatFromCurve(FloatCurve=DataInterface() {propertyValues {Curve {keys [
         <at 0.1769 0.9943 RCIM_Cubic>
         {arriveTangent=-1.9828, interpMode=RCIM_Cubic, leaveTangent=-1.9828, tangentMode=RCTM_User, time=1.0011, value=0.002}
        ]}}})
   )
   &SolveForcesAndVelocity(Clamp Velocity=false, Limit Acceleration=false)
  ]
  rendererClasses [&NiagaraRibbonRendererProperties]
  renderers [{rendererClass=&NiagaraRibbonRendererProperties}]
 }}
summary {
 systemName=NS_W03_Q3_RadiowaveTrail
 emitters [{emitterName=Beam, rendererClasses [&NiagaraRibbonRendererProperties]}]
 userVariables [
  User.BeamEndLocation {type=Vector3f, description="", defaultValue {struct=Vector3f, value=<0 0 250>}}
  User.MasterAlpha {type=NiagaraFloat, defaultValue=1, description=""}
 ]
}
